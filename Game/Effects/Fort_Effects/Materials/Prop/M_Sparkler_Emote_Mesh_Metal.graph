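# Default Lit material for the metal part of a sparkler emote prop, set up for use on mesh particles, including Niagara.
# The surface is untextured, built only from a grey-default Color parameter and scalar constants.
# A UseCameraOffset static switch selects between a clamped camera offset, driven by CamOffset through the engine's CameraOffset world-position function, and a constant fallback.
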Material /Game/Effects/Fort_Effects/Materials/Prop/M_Sparkler_Emote_Mesh_Metal
graph {
 properties {bUsedWithMeshParticles=true, bUsedWithNiagaraMeshParticles=true, pixelDepthOffsetMode=PDOM_Legacy, usedShadingModels=Default Lit}
 nodes [
  Constant() {r=1}
  Constant() {r=0.4}
  StaticSwitchParameter(
   True=MaterialFunctionCall(Offset Amount=ScalarParameter() {parameterName=CamOffset}) {materialFunction=/Engine/Functions/Engine_MaterialFunctions02/WorldPositionOffset/CameraOffset}.Clamped Camera Offset
   False=Constant()
  ) {parameterName=UseCameraOffset}
  VectorParameter() {parameterName=Color, defaultValue=<rgba 0.4 0.4 0.4 1>}
 ]
}
